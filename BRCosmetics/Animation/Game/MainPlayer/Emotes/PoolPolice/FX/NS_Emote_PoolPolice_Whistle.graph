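# Niagara system for the whistle effect in the PoolPolice main-player emote.
# A single Ripples emitter is declared, with no renderers or renderer classes listed for it.
# System scripts use SystemState and a spawn-time variable assignment, and the FNTimeOfDayColor dynamic input supplies a time-of-day color.
# The system exposes no user variables and uses no data interfaces.
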NiagaraSystem /BRCosmetics/Animation/Game/MainPlayer/Emotes/PoolPolice/FX/NS_Emote_PoolPolice_Whistle
dependencies {
 usedDataInterfaces []
 usedDynamicInputs [&FNTimeOfDayColor]
 usedModules [
  &self:SystemSpawnScript.SystemScriptSource.SystemScriptGraph.NiagaraNodeAssignment_1.SetVariables_#1
  &SystemState
 ]
}
emitters {Ripples {rendererClasses [], renderers []}}
summary {systemName=NS_Emote_PoolPolice_Whistle, emitters [{emitterName=Ripples, rendererClasses []}], userVariables []}
